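# Physics asset for the Meteorwomen suit on the female medium player body, covering its front and back dangling zippers.
# Each zipper pull capsule is constrained to a small master zipper capsule with swing limits and a tight twist limit.
# Both master zippers and a right thigh capsule link to the pelvis with free angular motion, with pelvis and thigh apparently serving as body collision for the zippers.
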PhysicsAsset /BRCosmetics/Characters/Player/Female/Medium/Bodies/F_MED_Meteorwomen_Suit/Meshes/F_MED_Meteorwomen_Suit_Physics
bodies {
 dyn_mstr_zipper_bk [{
   shapeType=Capsule
   length=0.646
   radius=0.5754
   shapeName=dyn_mstr_zipper_bk_capsule
   center=<-0.3954 0.0001 -0.312>
   rotation=<rot 0 89.8174 -83.5045>
  }]
 dyn_mstr_zipper_fr [{
   shapeType=Capsule
   length=0.6388
   radius=0.5433
   shapeName=dyn_mstr_zipper_fr_capsule
   center=<-0.3573 -0.2938 -0.1801>
   rotation=<rot 0 129.5609 -83.9917>
  }]
 dyn_zipper_bk [{shapeType=Capsule, length=1.7906, radius=0.575, shapeName=dyn_zipper_bk_capsule, center=<2.1932 0 0>, rotation=<rot 0 90 -90>}]
 dyn_zipper_fr [{shapeType=Capsule, length=1.7906, radius=0.575, shapeName=dyn_zipper_bk_capsule, center=<2.1932 0 0>, rotation=<rot 0 90 -90>}]
 pelvis [{shapeType=Capsule, length=8.25, radius=10.25, center=<-3.9082 -2.2266 0>}]
 thigh_r [{shapeType=Capsule, length=16.2968, radius=6.8255, shapeName=thigh_r_capsule, center=<9.0447 0 0>, rotation=<rot 90 0 0>}]
}
constraints [
 {bone1Name=dyn_zipper_bk, bone2Name=dyn_mstr_zipper_bk, swing1LimitDegrees=70, swing2LimitDegrees=90, twistLimitDegrees=10}
 {bone1Name=dyn_zipper_fr, bone2Name=dyn_mstr_zipper_fr, swing1LimitDegrees=70, swing2LimitDegrees=90, twistLimitDegrees=10}
 {
  bone1Name=dyn_mstr_zipper_bk
  bone2Name=pelvis
  swing1LimitDegrees=45
  swing1Motion=Free
  swing2LimitDegrees=45
  swing2Motion=Free
  twistLimitDegrees=45
  twistMotion=Free
 }
 {
  bone1Name=dyn_mstr_zipper_fr
  bone2Name=pelvis
  swing1LimitDegrees=45
  swing1Motion=Free
  swing2LimitDegrees=45
  swing2Motion=Free
  twistLimitDegrees=45
  twistMotion=Free
 }
 {
  bone1Name=thigh_r
  bone2Name=pelvis
  swing1LimitDegrees=45
  swing1Motion=Free
  swing2LimitDegrees=45
  swing2Motion=Free
  twistLimitDegrees=45
  twistMotion=Free
 }
]
