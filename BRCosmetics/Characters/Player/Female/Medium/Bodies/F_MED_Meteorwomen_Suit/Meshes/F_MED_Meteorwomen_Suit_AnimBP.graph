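# Post-attachment Anim Blueprint for the female medium Meteorwomen suit body, layering deformation and secondary motion onto the attached parent's pose.
# The copied pose runs through a deform Control Rig, then weighted constraints drive front and back zipper master bones from pelvis, thigh twist and groin/glute bones.
# Zipper AnimDynamics nodes follow at zero alpha, so they are inactive.
# A final RigidBody node simulates the suit's physics asset in spine_01 space with damped component motion.
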AnimBlueprint /BRCosmetics/Characters/Player/Female/Medium/Bodies/F_MED_Meteorwomen_Suit/Meshes/F_MED_Meteorwomen_Suit_AnimBP
target_skeleton=&/F_MED_Meteorwomen_Suit_Skeleton
graphs [
 AnimGraph {nodes [
   n1 = Control Rig(Source=Copy Pose From Mesh() {bUseAttachedParent=true}) {
    alphaBoolBlend {blendOption=Linear}
    controlRigAssetReference {blueprintRigClass=&/F_MED_DeformRig_MeteorWomen_suit_CtrlRig.F_MED_DeformRig_MeteorWomen_suit_CtrlRig_C}
   }
   n2 = "Constraint - dyn_mstr_zipper_fr (PPP)"!(
    ComponentPose=Local To Component(LocalPose=@n1)
    Alpha=1
    ConstraintWeights_0=0.7573
    ConstraintWeights_1:float=0.0839
    ConstraintWeights_2:float=0.1589
   ) -> (Pose) {
    alphaBoolBlend {blendOption=Linear}
    boneToModify {boneName=dyn_mstr_zipper_fr}
    constraintSetup [{targetBone {boneName=pelvis}}, {targetBone {boneName=thigh_twist_01_r}}, {targetBone {boneName=deform_groin_r}}]
    constraintWeights [1,1,1]
   }
   n3 = "Constraint - dyn_mstr_zipper_bk (PPP)"!(ComponentPose=@n2, Alpha=1, ConstraintWeights_0=0.7573, ConstraintWeights_1:float=0.0839, ConstraintWeights_2:float=0.1589) -> (Pose) {
    alphaBoolBlend {blendOption=Linear}
    boneToModify {boneName=dyn_mstr_zipper_bk}
    constraintSetup [{targetBone {boneName=pelvis}}, {targetBone {boneName=thigh_twist_01_l}}, {targetBone {boneName=deform_glute_l}}]
    constraintWeights [1,1,1]
   }
   n4 = Anim Dynamics - Bone: dyn_zipper_bk(ComponentPose=@n3, Alpha=0, ExternalForce="0,0,0") {
    angularSpringConstant=5
    bAngularSpring=true
    alphaBoolBlend {blendOption=Linear}
    boundBone {boneName=dyn_zipper_bk}
    componentAppliedLinearAccClamp=<10000 10000 10000>
    componentLinearAccScale=<1 1 1>
    componentLinearVelScale=<1 1 1>
    physicsBodyDefinitions [{
      boundBone {boneName=dyn_zipper_bk}
      boxExtents=<20 5 5>
      constraintSetup {angularLimitsMax=<0 10 60>, angularLimitsMin=<0 -180 -60>, angularTarget=<0 0 0>}
      localJointOffset=<2 0 0>
     }]
    relativeSpaceBone {boneName=root}
   }
   n5 = Anim Dynamics - Bone: dyn_zipper_fr(ComponentPose=@n4, Alpha=0, ExternalForce="0,0,0") {
    angularSpringConstant=5
    bAngularSpring=true
    alphaBoolBlend {blendOption=Linear}
    boundBone {boneName=dyn_zipper_fr}
    componentAppliedLinearAccClamp=<10000 10000 10000>
    componentLinearAccScale=<1 1 1>
    componentLinearVelScale=<1 1 1>
    physicsBodyDefinitions [{
      boundBone {boneName=dyn_zipper_fr}
      boxExtents=<20 5 5>
      constraintSetup {angularLimitsMax=<0 10 60>, angularLimitsMin=<0 -180 -60>, angularTarget=<0 0 0>}
      localJointOffset=<2 0 0>
     }]
    relativeSpaceBone {boneName=root}
   }
   n6 = RigidBody(ComponentPose=@n5) {
    bOverrideWorldGravity=true
    overridePhysicsAsset=&/F_MED_Meteorwomen_Suit_Physics
    simulationSpace=BaseBoneSpace
    alphaBoolBlend {blendOption=Linear}
    baseBoneRef {boneName=spine_01}
    componentAppliedLinearAccClamp=<1000 1000 1000>
    componentLinearAccScale=<0.4 0.4 0.4>
    componentLinearVelScale=<0.4 0.4 0.4>
    overrideWorldGravity {z=-980}
   }
   Output Pose(Result=Component To Local(ComponentPose=@n6)) {layerGroup=None}
  ]}
 EventGraph {nodes [Event BlueprintUpdateAnimation()]}
]
variables []
