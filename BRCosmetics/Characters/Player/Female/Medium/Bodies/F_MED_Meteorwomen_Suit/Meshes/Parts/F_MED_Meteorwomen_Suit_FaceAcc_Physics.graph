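# Physics asset for the face accessory part of the female Meteorwomen suit cosmetic, defining a physics braid that hangs from the head.
# Seven capsules form the dyn_braid chain, linked by constraints with tight twist and limited swing, stiffer near the root at the second and third links.
# Capsules on the head, neck, clavicles and upper spine form the body shapes around the braid, joined by constraints with free swing and twist.
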PhysicsAsset /BRCosmetics/Characters/Player/Female/Medium/Bodies/F_MED_Meteorwomen_Suit/Meshes/Parts/F_MED_Meteorwomen_Suit_FaceAcc_Physics
bodies {
 clavicle_l [{shapeType=Capsule, length=8.75, radius=9, shapeName=clavicle_l_capsule, center=<13.7675 4.3572 1.9113>, rotation=<rot 90 0 0>}]
 clavicle_r [{shapeType=Capsule, length=8.75, radius=9, shapeName=clavicle_l_capsule, center=<-13.7675 -4.3572 -1.9113>, rotation=<rot 90 0 0>}]
 dyn_braid_01 [{shapeType=Capsule, length=2.2592, radius=2.0368, shapeName=dyn_braid_01_capsule, center=<1 0 0>, rotation=<rot 5.2746 83.9251 -90>}]
 dyn_braid_02 [{shapeType=Capsule, length=2.5, radius=2, shapeName=dyn_braid_02_capsule, center=<2 0 0>, rotation=<rot 90 0 0>}]
 dyn_braid_03 [{shapeType=Capsule, length=2.5, radius=2, shapeName=dyn_braid_03_capsule, center=<2 0 0>, rotation=<rot 90 0 0>}]
 dyn_braid_04 [{shapeType=Capsule, length=3, radius=2, shapeName=dyn_braid_04_capsule, center=<2 0 0>, rotation=<rot 90 0 0>}]
 dyn_braid_05 [{shapeType=Capsule, length=3, radius=2, shapeName=dyn_braid_05_capsule, center=<2 0 0>, rotation=<rot 90 0 0>}]
 dyn_braid_06 [{shapeType=Capsule, length=3, radius=2, shapeName=dyn_braid_06_capsule, center=<2 0 0>, rotation=<rot 90 0 0>}]
 dyn_braid_07 [{shapeType=Capsule, length=3, radius=2, shapeName=dyn_braid_07_capsule, center=<2 0 0>, rotation=<rot 90 0 0>}]
 head [{shapeType=Capsule, length=3.75, radius=8.5, shapeName=head_capsule, center=<-0.202 3.5179 0>, rotation=<rot 90 180 180>}]
 neck_01 [{shapeType=Capsule, length=10.005, radius=5, shapeName=neck_01_capsule, center=<0 0.4479 0>}]
 neck_02 [{shapeType=Capsule, length=10.005, radius=4, shapeName=neck_01_capsule}]
 spine_03 [{shapeType=Capsule, length=25.505, radius=10.5, shapeName=spine_03_capsule, center=<9.5256 1.274 0>, rotation=<rot 0 -90 0>}]
 spine_05 [{shapeType=Capsule, length=16.755, radius=9.255, shapeName=spine_05_capsule, center=<2.5115 0.5437 0>, rotation=<rot 0 -90 0>}]
}
constraints [
 {
  bone1Name=head
  bone2Name=neck_02
  swing1LimitDegrees=45
  swing1Motion=Free
  swing2LimitDegrees=45
  swing2Motion=Free
  twistLimitDegrees=45
  twistMotion=Free
 }
 {
  bone1Name=neck_01
  bone2Name=spine_05
  swing1LimitDegrees=45
  swing1Motion=Free
  swing2LimitDegrees=45
  swing2Motion=Free
  twistLimitDegrees=45
  twistMotion=Free
 }
 {
  bone1Name=clavicle_r
  bone2Name=spine_05
  swing1LimitDegrees=45
  swing1Motion=Free
  swing2LimitDegrees=45
  swing2Motion=Free
  twistLimitDegrees=45
  twistMotion=Free
 }
 {
  bone1Name=clavicle_l
  bone2Name=spine_05
  swing1LimitDegrees=45
  swing1Motion=Free
  swing2LimitDegrees=45
  swing2Motion=Free
  twistLimitDegrees=45
  twistMotion=Free
 }
 {
  bone1Name=spine_05
  bone2Name=spine_03
  swing1LimitDegrees=45
  swing1Motion=Free
  swing2LimitDegrees=45
  swing2Motion=Free
  twistLimitDegrees=45
  twistMotion=Free
 }
 {bone1Name=dyn_braid_07, bone2Name=dyn_braid_06, swing1LimitDegrees=45, swing2LimitDegrees=45, twistLimitDegrees=3}
 {bone1Name=dyn_braid_01, bone2Name=head, swing1LimitDegrees=45, swing2LimitDegrees=45, twistLimitDegrees=3}
 {bone1Name=dyn_braid_02, bone2Name=dyn_braid_01, swing1LimitDegrees=25, swing2LimitDegrees=25, twistLimitDegrees=3}
 {bone1Name=dyn_braid_03, bone2Name=dyn_braid_02, swing1LimitDegrees=35, swing2LimitDegrees=35, twistLimitDegrees=3}
 {bone1Name=dyn_braid_04, bone2Name=dyn_braid_03, swing1LimitDegrees=45, swing2LimitDegrees=45, twistLimitDegrees=3}
 {bone1Name=dyn_braid_05, bone2Name=dyn_braid_04, swing1LimitDegrees=45, swing2LimitDegrees=45, twistLimitDegrees=3}
 {bone1Name=dyn_braid_06, bone2Name=dyn_braid_05, swing1LimitDegrees=45, swing2LimitDegrees=45, twistLimitDegrees=3}
 {
  bone1Name=neck_02
  bone2Name=neck_01
  swing1LimitDegrees=45
  swing1Motion=Free
  swing2LimitDegrees=45
  swing2Motion=Free
  twistLimitDegrees=45
  twistMotion=Free
 }
]
