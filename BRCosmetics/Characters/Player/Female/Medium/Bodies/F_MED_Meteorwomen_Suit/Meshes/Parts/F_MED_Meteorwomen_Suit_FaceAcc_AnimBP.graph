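# Post-process style AnimBP for the Meteorwomen Suit face accessory, posing and simulating its braid on the player.
# The parent's copied pose runs through Control Rig, RigidBody physics without back bling, or weighted braid constraints and offsets with it.
# Further additive corrections, some driven by head yaw on the Alt body mesh, surround AnimDynamics on lower braid bones.
# EventGraph updates gravity override, backpack and neck alphas; two RigidBody branches are unconnected.
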AnimBlueprint /BRCosmetics/Characters/Player/Female/Medium/Bodies/F_MED_Meteorwomen_Suit/Meshes/Parts/F_MED_Meteorwomen_Suit_FaceAcc_AnimBP
target_skeleton=&/F_MED_Meteorwomen_Suit_FaceAcc_Skeleton
graphs [
 AnimGraph {nodes [
   n1 = Get Gravity Override()
   n2 = Get noBackpack() -> (noBackpack:double)
   RigidBody!(ComponentPose, Alpha=@n2, OverrideWorldGravity:Vector=@n1, ExternalForce) -> (Pose) {
    bClampLinearTranslationLimitToRefPose=true
    bOverrideWorldGravity=true
    overridePhysicsAsset=&/F_MED_Meteorwomen_Suit_FaceAcc_Physics
    simulationSpace=BaseBoneSpace
    alphaBoolBlend {blendOption=Linear}
    baseBoneRef {boneName=root}
    componentAppliedLinearAccClamp=<3000 3000 3000>
    componentLinearAccScale=<0.25 0.25 0.35>
    componentLinearVelScale=<0.25 0.25 0.35>
   }
   n3 = "Constraint - dyn_braid_mstr (PP)"!(ComponentPose, Alpha=0, ConstraintWeights_0=0.8, ConstraintWeights_1:float=0.2) -> (Pose) {
    alphaBoolBlend {blendOption=Linear}
    boneToModify {boneName=dyn_braid_mstr}
    constraintSetup [{targetBone {boneName=head}}, {targetBone {boneName=dyn_braid_01}}]
    constraintWeights [1,1]
   }
   RigidBody!(ComponentPose=@n3, Alpha=@n2, OverrideWorldGravity=@n1, ExternalForce, ComponentLinearAccScale:Vector="0,0,0") -> (Pose) {
    bClampLinearTranslationLimitToRefPose=true
    bOverrideWorldGravity=true
    overridePhysicsAsset=&F_MED_Meteorwomen_Suit_FaceAcc_Physics
    simulationSpace=BaseBoneSpace
    alphaBoolBlend {blendOption=Linear}
    baseBoneRef {boneName=spine_05}
    componentAppliedLinearAccClamp=<2500 2500 2500>
    componentLinearAccScale=<0.4 0.4 0.5>
    componentLinearVelScale=<0.05 0.05 0.1>
    overrideWorldGravity {z=-980}
    simSpaceSettings {
     maxAngularAcceleration=1000
     maxAngularVelocity=3
     maxLinearAcceleration=700
     maxLinearVelocity=500
     velocityScaleZ=0
     worldAlpha=1
     externalLinearDragV=<0.2 0.2>
    }
   }
   n4 = Control Rig(Source=Copy Pose From Mesh() {bUseAttachedParent=true}) {
    alphaBoolBlend {blendOption=Linear}
    controlRigAssetReference {blueprintRigClass=&/F_MED_Meteorwomen_Suit_FaceAcc_CtrlRig.F_MED_Meteorwomen_Suit_FaceAcc_CtrlRig_C}
   }
   n5 = RigidBody!(ComponentPose=Local To Component(LocalPose=@n4), Alpha=Get noBackpack(), OverrideWorldGravity=Get Gravity Override(), ExternalForce) -> (Pose) {
    bClampLinearTranslationLimitToRefPose=true
    bOverrideWorldGravity=true
    overridePhysicsAsset=&F_MED_Meteorwomen_Suit_FaceAcc_Physics
    simulationSpace=BaseBoneSpace
    alphaBoolBlend {blendOption=Linear}
    baseBoneRef {boneName=spine_05}
    componentAppliedLinearAccClamp=<3000 3000 3000>
    componentLinearAccScale=<0.02 0.02 0.02>
    componentLinearVelScale=<0.02 0.02 0.02>
    simSpaceSettings {velocityScaleZ=0.73, worldAlpha=0.66}
   }
   n6 = Get wearingBackpack() -> (wearingBackpack:double)
   n7 = "Constraint - dyn_braid_01 (PP)"!(ComponentPose=@n5, Alpha=@n6, ConstraintWeights_0=0.8, ConstraintWeights_1:float=0.2) -> (Pose) {
    alphaBoolBlend {blendOption=Linear}
    boneToModify {boneName=dyn_braid_01}
    constraintSetup [{targetBone {boneName=head}}, {targetBone {boneName=spine_05}}]
    constraintWeights [1,1]
   }
   n8 = "Constraint - dyn_braid_02 (PP)"!(ComponentPose=@n7, Alpha=@n6, ConstraintWeights_0=0.5, ConstraintWeights_1:float=0.5) -> (Pose) {
    alphaBoolBlend {blendOption=Linear}
    boneToModify {boneName=dyn_braid_02}
    constraintSetup [{targetBone {boneName=head}}, {targetBone {boneName=spine_05}}]
    constraintWeights [1,1]
   }
   n9 = "Constraint - dyn_braid_03 (PPU)"!(ComponentPose=@n8, Alpha=@n6, ConstraintWeights_0=0.3, ConstraintWeights_1:float=0.7, ConstraintWeights_2:float=0) -> (Pose) {
    alphaBoolBlend {blendOption=Linear}
    boneToModify {boneName=dyn_braid_03}
    constraintSetup [{targetBone {boneName=head}}, {targetBone {boneName=spine_05}}, {transformType=LookAt, targetBone {boneName=dyn_braid_04}}]
    constraintWeights [1,1,1]
   }
   n10 = "Constraint - dyn_braid_04 (PPP)"!(ComponentPose=@n9, Alpha=@n6, ConstraintWeights_0=0.7, ConstraintWeights_1:float=0.1, ConstraintWeights_2:float=0.2) -> (Pose) {
    alphaBoolBlend {blendOption=Linear}
    boneToModify {boneName=dyn_braid_04}
    constraintSetup [{targetBone {boneName=spine_05}}, {targetBone {boneName=head}}, {targetBone {boneName=spine_04}}]
    constraintWeights [1,1,1]
   }
   n11 = "Constraint - dyn_braid_05 (PPP)"!(ComponentPose=@n10, Alpha=@n6, ConstraintWeights_0=0.475, ConstraintWeights_1:float=0.475, ConstraintWeights_2:float=0.05) -> (Pose) {
    alphaBoolBlend {blendOption=Linear}
    boneToModify {boneName=dyn_braid_05}
    constraintSetup [{targetBone {boneName=spine_04}}, {targetBone {boneName=spine_05}}, {targetBone {boneName=head}}]
    constraintWeights [1,1,1]
   }
   n12 = "Constraint - dyn_braid_06 (PP)"!(ComponentPose=@n11, Alpha=@n6, ConstraintWeights_0=0.5, ConstraintWeights_1:float=0.5) -> (Pose) {
    alphaBoolBlend {blendOption=Linear}
    boneToModify {boneName=dyn_braid_06}
    constraintSetup [{targetBone {boneName=spine_04}}, {targetBone {boneName=spine_05}}]
    constraintWeights [1,1]
   }
   n13 = "Constraint - dyn_braid_07 (PP)"!(ComponentPose=@n12, Alpha=@n6, ConstraintWeights_0=0.5, ConstraintWeights_1:float=0.5) -> (Pose) {
    alphaBoolBlend {blendOption=Linear}
    boneToModify {boneName=dyn_braid_07}
    constraintSetup [{targetBone {boneName=spine_05}}, {targetBone {boneName=spine_04}}]
    constraintWeights [1,1]
   }
   n14 = Get wearingBackpack()
   n15 = "Transform (Modify) Bone - Bone: dyn_braid_01"(ComponentPose=@n13, Alpha=@n14, Translation="0,0,-1", Rotation="0,0,0", Scale="1,1,1") {
    rotationMode=BMM_Additive
    scaleSpace=BCS_BoneSpace
    translationMode=BMM_Additive
    alphaBoolBlend {blendOption=Linear}
    boneToModify {boneName=dyn_braid_01}
    translation {z=-1}
   }
   n16 = "Transform (Modify) Bone - Bone: dyn_braid_03"(ComponentPose=@n15, Alpha=@n14, Translation="0,0,0", Rotation="0,0,0", Scale="1,1,1") {
    rotationMode=BMM_Additive
    scaleSpace=BCS_BoneSpace
    translationMode=BMM_Additive
    alphaBoolBlend {blendOption=Linear}
    boneToModify {boneName=dyn_braid_03}
   }
   n17 = "Transform (Modify) Bone - Bone: dyn_braid_05"(ComponentPose=@n16, Alpha=@n14, Translation="-0.000,1,0", Rotation="0,0,-10", Scale="1,1,1") {
    rotationMode=BMM_Additive
    scaleSpace=BCS_BoneSpace
    translationMode=BMM_Additive
    alphaBoolBlend {blendOption=Linear}
    boneToModify {boneName=dyn_braid_05}
    rotation {roll=-10}
    translation {y=1}
   }
   n18 = "Transform (Modify) Bone - Bone: dyn_braid_04"(ComponentPose=@n17, Alpha=1, Translation="0.000,-0.5,0.000", Rotation="0,0,0", Scale="1,1,1") {
    rotationMode=BMM_Additive
    scaleSpace=BCS_BoneSpace
    translationMode=BMM_Additive
    alphaBoolBlend {blendOption=Linear}
    boneToModify {boneName=dyn_braid_04}
    translation {y=-0.5}
   }
   n19 = "Transform (Modify) Bone - Bone: dyn_braid_05"(ComponentPose=@n18, Alpha=1, Translation="0.000,-1.5,0.000", Rotation="0,0,0", Scale="1,1,1") {
    rotationMode=BMM_Additive
    scaleSpace=BCS_BoneSpace
    translationMode=BMM_Additive
    alphaBoolBlend {blendOption=Linear}
    boneToModify {boneName=dyn_braid_05}
    translation {y=-1.5}
   }
   n20 = "Transform (Modify) Bone - Bone: dyn_braid_02"(ComponentPose=@n19, Alpha=Get Neck Alpha() -> (Neck Alpha:double), Translation="0,0,0", Rotation="-0.000,0.000,10.000", Scale="1,1,1") {
    rotationMode=BMM_Additive
    scaleSpace=BCS_BoneSpace
    translationMode=BMM_Additive
    alphaBoolBlend {blendOption=Linear}
    boneToModify {boneName=dyn_braid_02}
    rotation {roll=10}
   }
   n21 = "Transform (Modify) Bone - Bone: dyn_braid_04"(ComponentPose=@n20, Alpha=Get Neck Alpha(), Translation="0,0,-1", Rotation="-0.000,0.000,-15", Scale="1,1,1") {
    rotationMode=BMM_Additive
    scaleSpace=BCS_BoneSpace
    translationMode=BMM_Additive
    alphaBoolBlend {blendOption=Linear}
    boneToModify {boneName=dyn_braid_04}
    rotation {roll=-15}
    translation {z=-1}
   }
   n22 = "Transform (Modify) Bone - Bone: dyn_braid_05"(ComponentPose=@n21, Alpha=Get Neck Alpha(), Translation="0,0,0", Rotation="-0.000,0.000,-5", Scale="1,1,1") {
    rotationMode=BMM_Additive
    scaleSpace=BCS_BoneSpace
    translationMode=BMM_Additive
    alphaBoolBlend {blendOption=Linear}
    boneToModify {boneName=dyn_braid_05}
    rotation {roll=-5}
   }
   n23 = Anim Dynamics - Bone: dyn_braid_05(ComponentPose=@n22, Alpha=1, ExternalForce="0,0,0") {
    simulationSpace=BoneRelative
    alphaBoolBlend {blendOption=Linear}
    boundBone {boneName=dyn_braid_05}
    physicsBodyDefinitions [{
      boundBone {boneName=dyn_braid_05}
      boxExtents=<10 10 10>
      constraintSetup {angularLimitsMax=<0 0 0>, angularLimitsMin=<0 0 0>, angularTarget=<0 0 0>}
      localJointOffset=<0 0 0>
     }]
    relativeSpaceBone {boneName=spine_05}
   }
   n24 = Anim Dynamics - Bone: dyn_braid_04(ComponentPose=@n23, Alpha=1, ExternalForce="0,0,0") {
    angularSpringConstant=1
    bAngularSpring=true
    alphaBoolBlend {blendOption=Linear}
    boundBone {boneName=dyn_braid_04}
    componentAppliedLinearAccClamp=<3000 3000 30000>
    componentLinearAccScale=<0.01 0.01 0.01>
    componentLinearVelScale=<0.01 0.01 0.01>
    physicsBodyDefinitions [{
      boundBone {boneName=dyn_braid_04}
      boxExtents=<40 40 40>
      constraintSetup {angularLimitsMax=<20 0 60>, angularLimitsMin=<-20 0 -60>, angularTarget=<0 0 0>}
      localJointOffset=<2 0 0>
     }]
    relativeSpaceBone {boneName=spine_05}
   }
   n25 = Anim Dynamics - Bone: dyn_braid_05(ComponentPose=@n24, Alpha=1, ExternalForce="0,0,0") {
    angularSpringConstant=1
    bAngularSpring=true
    alphaBoolBlend {blendOption=Linear}
    boundBone {boneName=dyn_braid_05}
    componentAppliedLinearAccClamp=<3000 3000 30000>
    componentLinearAccScale=<0.01 0.01 0.01>
    componentLinearVelScale=<0.01 0.01 0.01>
    physicsBodyDefinitions [{
      boundBone {boneName=dyn_braid_05}
      boxExtents=<40 40 40>
      constraintSetup {angularLimitsMax=<20 0 60>, angularLimitsMin=<-20 0 -60>, angularTarget=<0 0 0>}
      localJointOffset=<2 0 0>
     }]
    relativeSpaceBone {boneName=spine_05}
   }
   n26 = Anim Dynamics - Bone: dyn_braid_06(ComponentPose=@n25, Alpha=1, ExternalForce="0,0,0") {
    angularSpringConstant=1
    bAngularSpring=true
    alphaBoolBlend {blendOption=Linear}
    boundBone {boneName=dyn_braid_06}
    componentAppliedLinearAccClamp=<3000 3000 30000>
    componentLinearAccScale=<0.01 0.01 0.01>
    componentLinearVelScale=<0.01 0.01 0.01>
    physicsBodyDefinitions [{
      boundBone {boneName=dyn_braid_06}
      boxExtents=<35 35 35>
      constraintSetup {angularLimitsMax=<20 0 60>, angularLimitsMin=<-20 0 -60>, angularTarget=<0 0 0>}
      localJointOffset=<2 0 0>
     }]
    relativeSpaceBone {boneName=spine_05}
   }
   n27 = "Transform (Modify) Bone - Bone: dyn_braid_01"(ComponentPose=@n26, Alpha=0, Translation="0,0,0", Rotation="0.000,0.000,-11.949", Scale="1,1,1") {
    rotationMode=BMM_Additive
    scaleSpace=BCS_BoneSpace
    translationMode=BMM_Additive
    alphaBoolBlend {blendOption=Linear}
    boneToModify {boneName=dyn_braid_01}
    rotation {roll=-11.949}
   }
   n28 = "Transform (Modify) Bone - Bone: dyn_braid_03"(ComponentPose=@n27, Alpha=1, Translation="0.000,-1,-0.000", Rotation="-0.000,0.000,7.087", Scale="1,1,1") {
    rotationMode=BMM_Additive
    scaleSpace=BCS_BoneSpace
    translationMode=BMM_Additive
    alphaBoolBlend {blendOption=Linear}
    boneToModify {boneName=dyn_braid_03}
    rotation {roll=7.087}
    translation {y=-1}
   }
   n29 = "Transform (Modify) Bone - Bone: dyn_braid_05"(ComponentPose=@n28, Alpha=1, Translation="-0.000,-1.185,0.000", Rotation="-0.000,0.000,12", Scale="1,1,1") {
    rotationMode=BMM_Additive
    scaleSpace=BCS_BoneSpace
    translationMode=BMM_Additive
    alphaBoolBlend {blendOption=Linear}
    boneToModify {boneName=dyn_braid_05}
    rotation {roll=12}
    translation {y=-1.185}
   }
   n30 = "Transform (Modify) Bone - Bone: dyn_braid_04"(ComponentPose=@n29, Alpha=1, Translation="-0.000,-1.2,0.000", Rotation="-0.000,0.000,-8", Scale="1,1,1") {
    rotationMode=BMM_Additive
    scaleSpace=BCS_BoneSpace
    translationMode=BMM_Additive
    alphaBoolBlend {blendOption=Linear}
    boneToModify {boneName=dyn_braid_04}
    rotation {roll=-8}
    translation {y=-1.2}
   }
   Output Pose(Result=Component To Local(ComponentPose=@n30)) {layerGroup=None}
  ]}
 EventGraph {nodes [
   Event BlueprintUpdateAnimation()
   > n1 = Sequence!(execute) -> (then_0, then_1, then_2:Exec)
   > n2 = Gravity Override State Function DEPRECATED(CharacterPartAnimInstance=Self-Reference(), GameTimeInSeconds=GetGameTimeInSeconds(), GravityOverrideParams=Get Gravity Override Params())
   > Set Gravity Override(Gravity Override=@n2)
   n3 = CheckForBackbling(self=Self-Reference())
   n4 = OR Boolean(A=@n3, B=@n3.bIsBackblingACape)
   Set wearing backpack Bool(execute=@n1.then_1, wearing backpack Bool:bool=@n4) -> (Output_Get:bool)
   > n5 = Set wearingBackpack(wearingBackpack:double="To Float (Boolean)"(InBool=@n4)) -> (Output_Get:double)
   > Set noBackpack(noBackpack:double="To Float (Boolean)"(InBool="Not Equal (Float)"(A=@n5, B=1))) -> (Output_Get:double)
   > n6 = Set FrontEndAlpha(FrontEndAlpha:bool=AND Boolean(A=Get wearing backpack Bool() -> (wearing backpack Bool:bool), B=Get bIsInFrontEnd())) -> (Output_Get:bool)
   n7 = Cast To PlayerPawn_Athena(execute=@n1.then_2, Object=TryGetPawnOwner())
   n8 = GetSkeletalMeshForPartType(self=@n7)
   n9 = GetDisplayName(Object=@n8)
   n10 = Append(A=GetObjectName(Object=@n8), B=" ")
   > n11 = Set BattleSuit?(BattleSuit?:bool="Equal, Case Insensitive (String)"(A=Split(SourceString=@n9, InStr=@n10, SearchDir=FromStart).RightS, B=F_MED_Meteorwomen_Alt)) -> (Output_Get:bool)
   Set BattleSuitAndBackFront(execute=@n6.then, BattleSuitAndBackFront:bool=AND Boolean(A=@n6, B=@n11)) -> (Output_Get:bool)
   n12 = GetOwnerSkeletalMeshComponent()
   n13 = BreakRotator(InRot=BreakTransform(InTransform=GetDeltaTransformFromRefPose(self=@n12, BoneName=head, BaseName=spine_05)).Rotation)
   Set Neck Alpha(
    execute=@n11.then
    Neck Alpha:double=SelectFloat(A=MapRangeClamped(Value=@n13.Yaw, InRangeA=-10, InRangeB=-40), B=0, bPickA=Get BattleSuit?() -> (BattleSuit?:bool))
   ) -> (Output_Get:double)
  ]}
]
variables [
 BattleSuit? {type=boolean}
 BattleSuitAndBackFront {type=boolean}
 FrontEndAlpha {type=boolean}
 Gravity Override {type=Vector}
 Gravity Override Params {
  type=GravityOverrideParamsStruct
  default {
   skydive_DivingDown {gravityOverride=<-1500 0 1500>}
   skydive_FloatingDown {gravityOverride=<0 -700 500>}
   skydive_GoingUp {gravityOverride=<0 -700 300>}
  }
 }
 Neck Alpha {type=number}
 noBackpack {type=number}
 wearing backpack Bool {type=boolean}
 wearingBackpack {type=number}
]
